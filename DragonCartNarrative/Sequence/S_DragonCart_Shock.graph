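# Level Sequence for the DragonCart narrative event in which an orb delivers three escalating shocks to a shard, which charges up and bursts into a powered state.
# Keyed Niagara parameters synchronize orb charge, blasts, shockwaves and electricity, then switch on the shard heart, gates and powered splines.
# The orb layers rotational wobbles over its main motion while its shield flickers.
# Timed camera shakes, force feedback and event triggers punctuate impacts, alongside a player-repulsor volume.
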LevelSequence /DragonCartNarrative/Sequence/S_DragonCart_Shock
bindings [
 NiagaraComponent0 {
  id=#1
  parent_id=#2
  possessed_class=NiagaraComponent
  tracks [{
    type=MovieSceneNiagaraFloatParameterTrack
    display_name=Intensity
    sections [{
      type=MovieSceneFloatSection
      channels [None {
        type=MovieSceneScriptingFloatChannel
        default=1
        keys [[173,1],[244.1,0],[840,0.7],[1110,0.7],[1379,0.3],[1380,0.65],[1650,0.65],[2219,0.015],[2220,1],[2511.0825,1],[2682.935,0]]
       }]
     }]
   }]
 }
 NS_DragonCart_Shard_Heart {
  id=#3
  parent_id=#4
  possessed_class=NiagaraComponent
  tracks [
   {
    type=MovieSceneVisibilityTrack
    sections [{
      type=MovieSceneVisibilitySection
      channels [None {type=MovieSceneScriptingBoolChannel, default=false, keys [[2347.615,false],[2929.225,true]]}]
     }]
   }
   {
    type=MovieSceneParticleTrack
    display_name=FX System
    sections [{
      type=MovieSceneParticleSection
      channels [None {type=MovieSceneScriptingParticleChannel, keys [[0, "<ParticleKey.DEACTIVATE: 1>"], [2931, "<ParticleKey.ACTIVATE: 0>"]]}]
     }]
   }
  ]
 }
 NiagaraComponent0 {
  id=#5
  parent_id=#6
  possessed_class=NiagaraComponent
  tracks [{
    type=MovieSceneNiagaraFloatParameterTrack
    display_name=Intensity
    sections [{
      type=MovieSceneFloatSection
      channels [None {
        type=MovieSceneScriptingFloatChannel
        default=0.015
        keys [[173,0.4],[244.1,0],[840,0.7],[1110,0.7],[1379,0.3],[1380,0.65],[1650,0.65],[2219,0.015],[2220,1],[2511.0825,1],[2682.935,0]]
       }]
     }]
   }]
 }
 NS_DragonCart_Shard_Heart_Burst {
  id=#7
  possessed_class=NiagaraActor
  tracks [
   {
    type=MovieSceneSpawnTrack
    sections [{type=MovieSceneSpawnSection, channels [None {type=MovieSceneScriptingBoolChannel, default=false, keys [[2347.615,false],[2929.225,true]]}]}]
   }
   {
    type=MovieScene3DTransformTrack
    sections [{
      type=MovieScene3DTransformSection
      channels [
       Location.X {type=MovieSceneScriptingDoubleChannel, default=-11011}
       Location.Y {type=MovieSceneScriptingDoubleChannel, default=0}
       Location.Z {type=MovieSceneScriptingDoubleChannel, default=3936.7445}
       Rotation.X {type=MovieSceneScriptingDoubleChannel, default=0}
       Rotation.Y {type=MovieSceneScriptingDoubleChannel, default=-0.3223}
       Rotation.Z {type=MovieSceneScriptingDoubleChannel, default=0}
       Scale.X {type=MovieSceneScriptingDoubleChannel, default=1}
       Scale.Y {type=MovieSceneScriptingDoubleChannel, default=1}
       Scale.Z {type=MovieSceneScriptingDoubleChannel, default=1}
      ]
     }]
   }
  ]
 }
 NiagaraComponent0 {
  id=#8
  parent_id=#9
  possessed_class=NiagaraComponent
  tracks [{
    type=MovieSceneNiagaraFloatParameterTrack
    display_name=Intensity
    sections [{
      type=MovieSceneFloatSection
      channels [None {
        type=MovieSceneScriptingFloatChannel
        default=0
        keys [[173,0],[244.1,0],[840,0.7],[1110,0.7],[1379,0.3],[1380,0.65],[1650,0.65],[2219,0.015],[2220,1],[2511.0825,1],[2682.935,0]]
       }]
     }]
   }]
 }
 NS_DragonCart_Shard_Electricity_8 {
  id=#6
  possessed_class=NiagaraActor
  tracks [
   {type=MovieSceneSpawnTrack, sections [{type=MovieSceneSpawnSection, channels [None {type=MovieSceneScriptingBoolChannel, default=true}]}]}
   {
    type=MovieScene3DTransformTrack
    sections [{
      type=MovieScene3DTransformSection
      channels [
       Location.X {type=MovieSceneScriptingDoubleChannel, default=-9858.8884}
       Location.Y {type=MovieSceneScriptingDoubleChannel, default=-28.614}
       Location.Z {type=MovieSceneScriptingDoubleChannel, default=2890.4319}
       Rotation.X {type=MovieSceneScriptingDoubleChannel, default=-2.9216}
       Rotation.Y {type=MovieSceneScriptingDoubleChannel, default=53.0101}
       Rotation.Z {type=MovieSceneScriptingDoubleChannel, default=-176.3535}
       Scale.X {type=MovieSceneScriptingDoubleChannel, default=1}
       Scale.Y {type=MovieSceneScriptingDoubleChannel, default=1}
       Scale.Z {type=MovieSceneScriptingDoubleChannel, default=1}
      ]
     }]
   }
  ]
 }
 Capsule - Repulsing {
  id=#10
  parent_id=#11
  possessed_class=CapsuleComponent
  tracks [
   {
    type=MovieSceneFloatTrack
    display_name=Capsule Half Height
    sections [{type=MovieSceneFloatSection, channels [None {type=MovieSceneScriptingFloatChannel, default=1788.6147, const [1788.6147,1]}]}]
   }
   {
    type=MovieSceneFloatTrack
    display_name=Capsule Radius
    sections [{type=MovieSceneFloatSection, channels [None {type=MovieSceneScriptingFloatChannel, default=288.5191, const [288.5191,1]}]}]
   }
   {
    type=MovieSceneFloatTrack
    display_name=Line Thickness
    sections [{type=MovieSceneFloatSection, channels [None {type=MovieSceneScriptingFloatChannel, default=80.0374, const [80.0374,1]}]}]
   }
  ]
 }
 NiagaraComponent0 {
  id=#12
  parent_id=#13
  possessed_class=NiagaraComponent
  tracks [
   {
    type=MovieSceneNiagaraFloatParameterTrack
    display_name=Intensity
    sections [{type=MovieSceneFloatSection, channels [None {type=MovieSceneScriptingFloatChannel, default=1.3, keys [[173,1.3],[244.1,0]]}]}]
   }
   {
    type=MovieSceneParticleTrack
    display_name=FX System
    sections [{
      type=MovieSceneParticleSection
      channels [None {type=MovieSceneScriptingParticleChannel, keys [[0, "<ParticleKey.ACTIVATE: 0>"], [285, "<ParticleKey.DEACTIVATE: 1>"]]}]
     }]
   }
   {
    type=MovieSceneNiagaraFloatParameterTrack
    display_name=Length
    sections [{type=MovieSceneFloatSection, channels [None {type=MovieSceneScriptingFloatChannel, default=1.3}]}]
   }
   {type=MovieScene3DAttachTrack, sections [{type=MovieScene3DAttachSection}]}
   {
    type=MovieSceneNiagaraFloatParameterTrack
    display_name=Lightning_Z_Offset
    sections [{type=MovieSceneFloatSection, channels [None {type=MovieSceneScriptingFloatChannel, default=-4100}]}]
   }
   {
    type=MovieSceneNiagaraBoolParameterTrack
    display_name=SpawnSmallLightningLFare
    sections [{type=MovieSceneBoolSection, channels [None {type=MovieSceneScriptingBoolChannel, default=false}]}]
   }
  ]
 }
 NS_DragonCart_Shard_Electricity_Long_DayOne6 {id=#14, possessed_class=NiagaraActor, tracks []}
 NS_DragonCart_OrbBlast_Looping_ForDayTwo {
  id=#15
  possessed_class=NiagaraActor
  tracks [
   {type=MovieSceneSpawnTrack, sections [{type=MovieSceneSpawnSection, channels [None {type=MovieSceneScriptingBoolChannel, default=true}]}]}
   {
    type=MovieScene3DTransformTrack
    sections [{
      type=MovieScene3DTransformSection
      channels [
       Location.X {type=MovieSceneScriptingDoubleChannel, default=-5045}
       Location.Y {type=MovieSceneScriptingDoubleChannel, default=0}
       Location.Z {type=MovieSceneScriptingDoubleChannel, default=-5337.1221}
       Rotation.X {type=MovieSceneScriptingDoubleChannel, default=0}
       Rotation.Y {type=MovieSceneScriptingDoubleChannel, default=0}
       Rotation.Z {type=MovieSceneScriptingDoubleChannel, default=0}
       Scale.X {type=MovieSceneScriptingDoubleChannel, default=1}
       Scale.Y {type=MovieSceneScriptingDoubleChannel, default=1}
       Scale.Z {type=MovieSceneScriptingDoubleChannel, default=1}
      ]
     }]
   }
  ]
 }
 NS_DragonCart_Shard_Electricity_10 {
  id=#16
  possessed_class=NiagaraActor
  tracks [
   {type=MovieSceneSpawnTrack, sections [{type=MovieSceneSpawnSection, channels [None {type=MovieSceneScriptingBoolChannel, default=true}]}]}
   {
    type=MovieScene3DTransformTrack
    sections [{
      type=MovieScene3DTransformSection
      channels [
       Location.X {type=MovieSceneScriptingDoubleChannel, default=-10222.8555}
       Location.Y {type=MovieSceneScriptingDoubleChannel, default=-717.8783}
       Location.Z {type=MovieSceneScriptingDoubleChannel, default=2897.449}
       Rotation.X {type=MovieSceneScriptingDoubleChannel, default=-0.5806}
       Rotation.Y {type=MovieSceneScriptingDoubleChannel, default=51.3822}
       Rotation.Z {type=MovieSceneScriptingDoubleChannel, default=119.957}
       Scale.X {type=MovieSceneScriptingDoubleChannel, default=1}
       Scale.Y {type=MovieSceneScriptingDoubleChannel, default=1}
       Scale.Z {type=MovieSceneScriptingDoubleChannel, default=1}
      ]
     }]
   }
  ]
 }
 BGA_DCN_OrbActorOpen {
  id=#17
  possessed_class=BGA_DCN_OrbActor_C
  tracks [{
    type=MovieScene3DTransformTrack
    sections [
     {
      type=MovieScene3DTransformSection
      channels [
       Weight {type=MovieSceneScriptingFloatChannel, default=1}
       Location.X {type=MovieSceneScriptingDoubleChannel, default=0, const [0,3]}
       Location.Y {type=MovieSceneScriptingDoubleChannel, default=0, const [0,3]}
       Location.Z {type=MovieSceneScriptingDoubleChannel, default=0, const [0,3]}
       Rotation.X {
        type=MovieSceneScriptingDoubleChannel
        default=0
        thinned_from=38
        keys [[841,0],[867,0.2153],[897,-0.0902],[934,0.1074],[958,0.0013],[986,-0.0599],[1024,0.0091],[1060,-0.0307],[1126,-0.0022],[1380,-0.0003],[1406,0.2495],[1440,-0.105],[1477,0.1243],[1501,0.0013],[1529,-0.0698],[1567,0.0104],[1603,-0.0358],[1669,-0.0028],[2218,-0.0002],[2246,0.3053],[2284,-0.1282],[2325,0.1039],[2352,-0.0108],[2383,-0.0738],[2425,0.0129],[2465,-0.0436],[2539,-0.0033],[2947,0],[2973,0.3925],[3002,-0.0324],[3034,0.2424],[3056,0.2289],[3085,0.1204],[3107,0.0407],[3134,0.1032],[3157,0],[3475,0]]
       }
       Rotation.Y {
        type=MovieSceneScriptingDoubleChannel
        default=0
        thinned_from=41
        keys [[841,0],[872,0.0767],[909,-0.0395],[945,0.0898],[983,-0.029],[1014,0.0145],[1045,-0.0078],[1071,0.0084],[1100,0],[1380,0],[1411,0.089],[1452,-0.0458],[1488,0.1042],[1526,-0.0336],[1557,0.0168],[1588,-0.009],[1614,0.0098],[1643,0],[2218,0],[2252,0.1089],[2298,-0.056],[2338,0.0919],[2380,-0.0411],[2414,0.0206],[2449,-0.011],[2477,0.0119],[2510,0],[2947,0],[2978,0.1067],[3015,-0.0549],[3048,0.1249],[3081,-0.0403],[3112,0.0202],[3143,-0.0108],[3169,0.0117],[3198,0],[3475,0]]
       }
       Rotation.Z {
        type=MovieSceneScriptingDoubleChannel
        default=0
        keys [[841,0],[871,-0.0241],[907,0.0504],[948,-0.0298],[988,0.0401],[1011,-0.0279],[1036,0.0151],[1063,-0.0093],[1090,0.0172],[1126,0],[1380,0],[1410,-0.0279],[1450,0.0585],[1491,-0.0346],[1531,0.0465],[1554,-0.0324],[1579,0.0175],[1606,-0.0108],[1633,0.0199],[1669,0],[2218,0],[2251,-0.0341],[2295,0.0715],[2341,-0.0423],[2385,0.0569],[2411,-0.0396],[2439,0.0214],[2469,-0.0132],[2499,0.0244],[2539,0],[2947,0],[2977,-0.0335],[3013,0.0701],[3051,-0.0414],[3086,0.0558],[3109,-0.0388],[3134,0.021],[3161,-0.013],[3188,0.0239],[3224,0],[3475,0]]
       }
       Scale.X {type=MovieSceneScriptingDoubleChannel, default=0, const [0,3]}
       Scale.Y {type=MovieSceneScriptingDoubleChannel, default=0, const [0,3]}
       Scale.Z {type=MovieSceneScriptingDoubleChannel, default=0, const [0,3]}
      ]
     }
     {
      type=MovieScene3DTransformSection
      channels [
       Location.X {type=MovieSceneScriptingDoubleChannel, default=-11011, keys [[300,-11011],[3304,-11011.0079]]}
       Location.Y {type=MovieSceneScriptingDoubleChannel, default=0, const [0,2]}
       Location.Z {
        type=MovieSceneScriptingDoubleChannel
        default=11136
        thinned_from=47
        keys [[0,11136],[195,11124.1299],[286,11053.2033],[626,10175.0108],[745,10233.4983],[841,10197.9966],[909,10456.7103],[923,10467.985],[973,10440.5199],[1028,10441.4964],[1138,10541.7355],[1202,10579.0049],[1228,10576.9314],[1325,10518.2387],[1381,10514.1255],[1429,10780.472],[1455,10833.8112],[1533,10793.1053],[1650,10894.863],[1714,10873.3195],[1853,11000.9549],[1914,10994.7167],[2086,10924.4324],[2217,10857.6942],[2277,11259.7036],[2294,11311.727],[2315,11337.1094],[2373,11362.3563],[2544,11528.7977],[2713,11498.675],[3047,10317.8776],[3111,9939.1848],[3425,10047.032],[3654,9910.5407],[3841,9906.6911]]
       }
       Rotation.X {
        type=MovieSceneScriptingDoubleChannel
        default=0
        thinned_from=14
        keys [[300,0],[2651.605,0],[2865,1.111],[2938,2.1221],[3009,6.7531],[3053,9.38],[3093,10.5166],[3218,9.0137],[3307,9.7459],[3407,9.0636]]
       }
       Rotation.Y {
        type=MovieSceneScriptingDoubleChannel
        default=0
        thinned_from=10
        keys [[300,0],[2651.605,0],[2745,-0.2318],[2823,-2.2641],[2899,-6.3192],[2989,-12.1755],[3059,-14.1146],[3382,-12.4463],[3511,-12.8442]]
       }
       Rotation.Z {
        type=MovieSceneScriptingDoubleChannel
        default=90
        thinned_from=78
        keys [[300,90],[895,89.919],[1386,89.9986],[1430,89.8577],[2234,89.9986],[2280,89.9052],[2315,90.0729],[2936,90.0001],[3007,88.1541],[3048,87.5938],[3131,87.3239],[3528,88.181]]
       }
       Scale.X {type=MovieSceneScriptingDoubleChannel, default=0.75, const [0.75,4]}
       Scale.Y {type=MovieSceneScriptingDoubleChannel, default=0.75, const [0.75,4]}
       Scale.Z {type=MovieSceneScriptingDoubleChannel, default=0.75, const [0.75,4]}
      ]
     }
    ]
   }]
 }
 NS_DragonCart_Shard_Electricity_7 {
  id=#18
  possessed_class=NiagaraActor
  tracks [
   {type=MovieSceneSpawnTrack, sections [{type=MovieSceneSpawnSection, channels [None {type=MovieSceneScriptingBoolChannel, default=true}]}]}
   {
    type=MovieScene3DTransformTrack
    sections [{
      type=MovieScene3DTransformSection
      channels [
       Location.X {type=MovieSceneScriptingDoubleChannel, default=-11324.2998}
       Location.Y {type=MovieSceneScriptingDoubleChannel, default=-897.2346}
       Location.Z {type=MovieSceneScriptingDoubleChannel, default=2872.26}
       Rotation.X {type=MovieSceneScriptingDoubleChannel, default=176.2594}
       Rotation.Y {type=MovieSceneScriptingDoubleChannel, default=42.128}
       Rotation.Z {type=MovieSceneScriptingDoubleChannel, default=79.42}
       Scale.X {type=MovieSceneScriptingDoubleChannel, default=1}
       Scale.Y {type=MovieSceneScriptingDoubleChannel, default=1}
       Scale.Z {type=MovieSceneScriptingDoubleChannel, default=1}
      ]
     }]
   }
  ]
 }
 BP_FakeCollisionVolume_and_PlayerRepulsor {
  id=#11
  possessed_class=BP_FakeCollisionVolume_and_PlayerRepulsor_C
  tracks [
   {type=MovieSceneSpawnTrack, sections [{type=MovieSceneSpawnSection, channels [None {type=MovieSceneScriptingBoolChannel, default=true}]}]}
   {
    type=MovieScene3DTransformTrack
    sections [{
      type=MovieScene3DTransformSection
      channels [
       Location.X {type=MovieSceneScriptingDoubleChannel, default=-11010.1319, const [-11010.1318,1]}
       Location.Y {type=MovieSceneScriptingDoubleChannel, default=6.44, const [6.44,1]}
       Location.Z {type=MovieSceneScriptingDoubleChannel, default=5673.1027, const [5673.1025,1]}
       Rotation.X {type=MovieSceneScriptingDoubleChannel, default=0, const [0,1]}
       Rotation.Y {type=MovieSceneScriptingDoubleChannel, default=0, const [0,1]}
       Rotation.Z {type=MovieSceneScriptingDoubleChannel, default=0, const [0,1]}
       Scale.X {type=MovieSceneScriptingDoubleChannel, default=1, const [1,1]}
       Scale.Y {type=MovieSceneScriptingDoubleChannel, default=1, const [1,1]}
       Scale.Z {type=MovieSceneScriptingDoubleChannel, default=1, const [1,1]}
      ]
     }]
   }
  ]
 }
 NS_DragonCart_ShardEnergy_01 {
  id=#19
  possessed_class=NiagaraActor
  tracks [
   {type=MovieSceneSpawnTrack, sections [{type=MovieSceneSpawnSection, channels [None {type=MovieSceneScriptingBoolChannel, default=true}]}]}
   {
    type=MovieScene3DTransformTrack
    sections [{
      type=MovieScene3DTransformSection
      channels [
       Location.X {type=MovieSceneScriptingDoubleChannel, default=-11011}
       Location.Y {type=MovieSceneScriptingDoubleChannel, default=0}
       Location.Z {type=MovieSceneScriptingDoubleChannel, default=3936.7445}
       Rotation.X {type=MovieSceneScriptingDoubleChannel, default=0}
       Rotation.Y {type=MovieSceneScriptingDoubleChannel, default=0}
       Rotation.Z {type=MovieSceneScriptingDoubleChannel, default=0}
       Scale.X {type=MovieSceneScriptingDoubleChannel, default=1}
       Scale.Y {type=MovieSceneScriptingDoubleChannel, default=1}
       Scale.Z {type=MovieSceneScriptingDoubleChannel, default=1}
      ]
     }]
   }
  ]
 }
 ForceFeedback {id=#20, parent_id=#21, possessed_class=ForceFeedbackComponent, tracks []}
 NiagaraComponent0 {
  id=#22
  parent_id=#23
  possessed_class=NiagaraComponent
  tracks [{
    type=MovieSceneNiagaraFloatParameterTrack
    display_name=Intensity
    sections [{
      type=MovieSceneFloatSection
      channels [None {
        type=MovieSceneScriptingFloatChannel
        default=0.4
        keys [[173,0.4],[244.1,0],[840,0.7],[1110,0.7],[1379,0.3],[1380,0.65],[1650,0.65],[2219,0.015],[2220,1],[2511.0825,1],[2682.935,0]]
       }]
     }]
   }]
 }
 NS_DragonCart_Shard_Gate N {
  id=#24
  parent_id=#4
  possessed_class=NiagaraComponent
  tracks [{
    type=MovieSceneParticleTrack
    display_name=FX System
    sections [{
      type=MovieSceneParticleSection
      channels [None {type=MovieSceneScriptingParticleChannel, keys [[0, "<ParticleKey.DEACTIVATE: 1>"], [2931, "<ParticleKey.ACTIVATE: 0>"]]}]
     }]
   }]
 }
 NS_DragonCart_OrbBlast_Looping_FromDayOne {
  id=#13
  possessed_class=NiagaraActor
  tracks [{
    type=MovieScene3DTransformTrack
    sections [{
      type=MovieScene3DTransformSection
      channels [
       Location.X {type=MovieSceneScriptingDoubleChannel, default=-5045}
       Location.Y {type=MovieSceneScriptingDoubleChannel, default=0}
       Location.Z {type=MovieSceneScriptingDoubleChannel, default=-5385.6276}
       Rotation.X {type=MovieSceneScriptingDoubleChannel, default=0}
       Rotation.Y {type=MovieSceneScriptingDoubleChannel, default=0}
       Rotation.Z {type=MovieSceneScriptingDoubleChannel, default=0}
       Scale.X {type=MovieSceneScriptingDoubleChannel, default=1}
       Scale.Y {type=MovieSceneScriptingDoubleChannel, default=1}
       Scale.Z {type=MovieSceneScriptingDoubleChannel, default=1}
      ]
     }]
   }]
 }
 ShieldMesh {
  id=#25
  parent_id=#17
  possessed_class=StaticMeshComponent
  tracks [
   {
    type=MovieSceneVisibilityTrack
    sections [{
      type=MovieSceneVisibilitySection
      channels [None {
        type=MovieSceneScriptingBoolChannel
        default=true
        keys [[717.4325,true],[754.125,false],[759.755,true],[766.4375,false],[771.8825,true],[794.61,false],[797.22,true],[804.2925,false],[807.36,true],[812.1525,false],[816.4125,true],[841,false],[844.44,true],[850.905,false],[857.05,true],[860.67,false],[864.6775,true],[876.6425,false],[881.335,true],[890.4025,false],[895.425,true],[899.3475,false],[3122.035,true],[3124.5875,false],[3127.205,true],[3129.3075,false],[3134,true],[3142.6475,false],[3145.5925,true],[3149.35,false],[3152.68,true]]
       }]
     }]
   }
   {type=MovieSceneEventTrack, sections [{type=MovieSceneEventTriggerSection, channels [None {type=MovieSceneScriptingEventChannel}]}]}
  ]
 }
 NS_DragonCart_Shard_Electricity_Long_DayOne7 {id=#26, possessed_class=NiagaraActor, tracks []}
 NS_DragonCart_Shard_Electricity_6 {
  id=#27
  possessed_class=NiagaraActor
  tracks [
   {type=MovieSceneSpawnTrack, sections [{type=MovieSceneSpawnSection, channels [None {type=MovieSceneScriptingBoolChannel, default=true}]}]}
   {
    type=MovieScene3DTransformTrack
    sections [{
      type=MovieScene3DTransformSection
      channels [
       Location.X {type=MovieSceneScriptingDoubleChannel, default=-10659.9345}
       Location.Y {type=MovieSceneScriptingDoubleChannel, default=921.7044}
       Location.Z {type=MovieSceneScriptingDoubleChannel, default=2844.3294}
       Rotation.X {type=MovieSceneScriptingDoubleChannel, default=179.3237}
       Rotation.Y {type=MovieSceneScriptingDoubleChannel, default=40.6269}
       Rotation.Z {type=MovieSceneScriptingDoubleChannel, default=-117.5096}
       Scale.X {type=MovieSceneScriptingDoubleChannel, default=1}
       Scale.Y {type=MovieSceneScriptingDoubleChannel, default=1}
       Scale.Z {type=MovieSceneScriptingDoubleChannel, default=1}
      ]
     }]
   }
  ]
 }
 NS_DragonCart_Shard_Gate W {
  id=#28
  parent_id=#4
  possessed_class=NiagaraComponent
  tracks [{
    type=MovieSceneParticleTrack
    display_name=FX System
    sections [{
      type=MovieSceneParticleSection
      channels [None {type=MovieSceneScriptingParticleChannel, keys [[0, "<ParticleKey.DEACTIVATE: 1>"], [2931, "<ParticleKey.ACTIVATE: 0>"]]}]
     }]
   }]
 }
 BP_DragonCart_ShardPowered_Splines {id=#4, possessed_class=BP_DragonCart_ShardPowered_Splines_C, tracks []}
 NiagaraComponent0 {
  id=#29
  parent_id=#27
  possessed_class=NiagaraComponent
  tracks [{
    type=MovieSceneNiagaraFloatParameterTrack
    display_name=Intensity
    sections [{
      type=MovieSceneFloatSection
      channels [None {
        type=MovieSceneScriptingFloatChannel
        default=0
        keys [[173,0],[244.1,0],[840,0.7],[1110,0.7],[1379,0.3],[1380,0.65],[1650,0.65],[2219,0.015],[2220,1],[2511.0825,1],[2682.935,0]]
       }]
     }]
   }]
 }
 Events_Shared_cameraForce_BP {
  id=#21
  possessed_class=Events_Shared_cameraForce_BP_C
  tracks [
   {
    type=MovieSceneSpawnTrack
    sections [{type=MovieSceneSpawnSection, channels [None {type=MovieSceneScriptingBoolChannel, default=false, keys [[372.205,false],[622.4,true]]}]}]
   }
   {
    type=MovieScene3DTransformTrack
    sections [{
      type=MovieScene3DTransformSection
      channels [
       Location.X {type=MovieSceneScriptingDoubleChannel, default=-11011}
       Location.Y {type=MovieSceneScriptingDoubleChannel, default=0}
       Location.Z {type=MovieSceneScriptingDoubleChannel, default=4576.3289}
       Rotation.X {type=MovieSceneScriptingDoubleChannel, default=0}
       Rotation.Y {type=MovieSceneScriptingDoubleChannel, default=0}
       Rotation.Z {type=MovieSceneScriptingDoubleChannel, default=0}
       Scale.X {type=MovieSceneScriptingDoubleChannel, default=1}
       Scale.Y {type=MovieSceneScriptingDoubleChannel, default=1}
       Scale.Z {type=MovieSceneScriptingDoubleChannel, default=1}
      ]
     }]
   }
   {
    type=MovieSceneIntegerTrack
    sections [{type=MovieSceneIntegerSection, channels [None {type=MovieSceneScriptingIntegerChannel, default=1, keys [[745.38,0],[1335.785,1]]}]}]
   }
   {
    type=MovieSceneObjectPropertyTrack
    display_name=Force Effect
    sections [{
      type=MovieSceneObjectPropertySection
      channels [None {
        type=MovieSceneScriptingObjectPathChannel
        default="<Object '&Events_Shared_small_FF' (0x49d75fdd280) Class 'ForceFeedbackEffect'>"
        keys [
         [745.38, "<Object '/Game/Weapons/ForceFeedback/FF_MissleLaunch' (0x49ddc5e6f20) Class 'ForceFeedbackEffect'>"]
         [
          2173.725
          "<Object '/FNGameplayCues/Weapons/ForceFeedback/Explosions/FF_Generic_ExplosionLong_Far' (0x49ddc5e78c0) Class 'ForceFeedbackEffect'>"
         ]
         [2681.935, "<Object '/Game/Weapons/ForceFeedback/FF_Jim_Charge_BuildUp' (0x49ddc5e6dd0) Class 'ForceFeedbackEffect'>"]
         [2928, "<Object '/Game/Weapons/ForceFeedback/Explosions/FF_Generic_ExplosionLarge_Near' (0x49ddc5e6e40) Class 'ForceFeedbackEffect'>"]
        ]
       }]
     }]
   }
   {
    type=MovieSceneEventTrack
    sections [{
      type=MovieSceneEventTriggerSection
      channels [None {
        type=MovieSceneScriptingEventChannel
        keys [
         [851, "<Struct 'MovieSceneEvent' (0x49e25463a20) {}>"]
         [1380, "<Struct 'MovieSceneEvent' (0x49e25463980) {}>"]
         [2220, "<Struct 'MovieSceneEvent' (0x49e254638e0) {}>"]
         [2749.2, "<Struct 'MovieSceneEvent' (0x49e25463840) {}>"]
         [2929.225, "<Struct 'MovieSceneEvent' (0x49e254637a0) {}>"]
        ]
       }]
     }]
   }
  ]
 }
 NiagaraComponent0 {
  id=#30
  parent_id=#31
  possessed_class=NiagaraComponent
  tracks [{
    type=MovieSceneNiagaraFloatParameterTrack
    display_name=Intensity
    sections [{
      type=MovieSceneFloatSection
      channels [None {
        type=MovieSceneScriptingFloatChannel
        default=0.4
        keys [[173,0.4],[244.1,0],[838,0.4],[839,0.4],[1110,0.7],[1380,0.65],[2220,1],[2511.0825,1],[2682.935,0]]
       }]
     }]
   }]
 }
 NS_DragonCart_Shard_Electricity_4 {
  id=#32
  possessed_class=NiagaraActor
  tracks [
   {type=MovieSceneSpawnTrack, sections [{type=MovieSceneSpawnSection, channels [None {type=MovieSceneScriptingBoolChannel, default=true}]}]}
   {
    type=MovieScene3DTransformTrack
    sections [{
      type=MovieScene3DTransformSection
      channels [
       Location.X {type=MovieSceneScriptingDoubleChannel, default=-12036.0849}
       Location.Y {type=MovieSceneScriptingDoubleChannel, default=-454.6329}
       Location.Z {type=MovieSceneScriptingDoubleChannel, default=2872.26}
       Rotation.X {type=MovieSceneScriptingDoubleChannel, default=175.2017}
       Rotation.Y {type=MovieSceneScriptingDoubleChannel, default=46.0561}
       Rotation.Z {type=MovieSceneScriptingDoubleChannel, default=29.6532}
       Scale.X {type=MovieSceneScriptingDoubleChannel, default=1}
       Scale.Y {type=MovieSceneScriptingDoubleChannel, default=1}
       Scale.Z {type=MovieSceneScriptingDoubleChannel, default=1}
      ]
     }]
   }
  ]
 }
 NiagaraComponent0 {
  id=#33
  parent_id=#34
  possessed_class=NiagaraComponent
  tracks [{
    type=MovieSceneNiagaraFloatParameterTrack
    display_name=Intensity
    sections [{
      type=MovieSceneFloatSection
      channels [None {
        type=MovieSceneScriptingFloatChannel
        default=0
        keys [[173,0],[244.1,0],[840,0.7],[1110,0.7],[1379,0.3],[1380,0.65],[1650,0.65],[2219,0.015],[2220,1],[2511.0825,1],[2682.935,0]]
       }]
     }]
   }]
 }
 CameraShakeSourceComponent {
  id=#35
  parent_id=#36
  possessed_class=CameraShakeSourceComponent
  tracks [
   {
    type=MovieSceneCameraShakeSourceShakeTrack
    display_name=Camera Shake
    sections [
     {type=MovieSceneCameraShakeSourceShakeSection, end_frame=3006, start_frame=2546}
     {type=MovieSceneCameraShakeSourceShakeSection, end_frame=1105, start_frame=645}
     {type=MovieSceneCameraShakeSourceShakeSection, end_frame=1684, start_frame=1224}
     {type=MovieSceneCameraShakeSourceShakeSection, end_frame=2466, start_frame=2006}
    ]
   }
   {type=MovieSceneCameraShakeSourceTriggerTrack, display_name=Camera Shake Trigger, sections [{type=MovieSceneCameraShakeSourceTriggerSection}]}
  ]
 }
 NS_DragonCart_Shard_Electricity_Long_DayOne10 {id=#2, possessed_class=NiagaraActor, tracks []}
 NiagaraComponent0 {
  id=#37
  parent_id=#32
  possessed_class=NiagaraComponent
  tracks [{
    type=MovieSceneNiagaraFloatParameterTrack
    display_name=Intensity
    sections [{
      type=MovieSceneFloatSection
      channels [None {
        type=MovieSceneScriptingFloatChannel
        default=0
        keys [[173,0],[244.1,0],[840,0.7],[1110,0.7],[1379,0.3],[1380,0.65],[1650,0.65],[2219,0.015],[2220,1],[2511.0825,1],[2682.935,0]]
       }]
     }]
   }]
 }
 BP_DragonCart_ShardPowered_Splines {
  id=#38
  possessed_class=BP_DragonCart_ShardPowered_Splines_C
  tracks [
   {
    type=MovieSceneSpawnTrack
    sections [{type=MovieSceneSpawnSection, channels [None {type=MovieSceneScriptingBoolChannel, default=false, keys [[2347.615,false],[2929.225,true]]}]}]
   }
   {
    type=MovieScene3DTransformTrack
    sections [{
      type=MovieScene3DTransformSection
      channels [
       Location.X {type=MovieSceneScriptingDoubleChannel, default=-11011}
       Location.Y {type=MovieSceneScriptingDoubleChannel, default=0}
       Location.Z {type=MovieSceneScriptingDoubleChannel, default=3936.7445}
       Rotation.X {type=MovieSceneScriptingDoubleChannel, default=0}
       Rotation.Y {type=MovieSceneScriptingDoubleChannel, default=0}
       Rotation.Z {type=MovieSceneScriptingDoubleChannel, default=0}
       Scale.X {type=MovieSceneScriptingDoubleChannel, default=1}
       Scale.Y {type=MovieSceneScriptingDoubleChannel, default=1}
       Scale.Z {type=MovieSceneScriptingDoubleChannel, default=1}
      ]
     }]
   }
  ]
 }
 NS_DragonCart_Shard_Electricity_12 {
  id=#9
  possessed_class=NiagaraActor
  tracks [
   {type=MovieSceneSpawnTrack, sections [{type=MovieSceneSpawnSection, channels [None {type=MovieSceneScriptingBoolChannel, default=true}]}]}
   {
    type=MovieScene3DTransformTrack
    sections [{
      type=MovieScene3DTransformSection
      channels [
       Location.X {type=MovieSceneScriptingDoubleChannel, default=-12036.0849}
       Location.Y {type=MovieSceneScriptingDoubleChannel, default=333.1426}
       Location.Z {type=MovieSceneScriptingDoubleChannel, default=2872.26}
       Rotation.X {type=MovieSceneScriptingDoubleChannel, default=175.2017}
       Rotation.Y {type=MovieSceneScriptingDoubleChannel, default=46.0561}
       Rotation.Z {type=MovieSceneScriptingDoubleChannel, default=-16.532}
       Scale.X {type=MovieSceneScriptingDoubleChannel, default=1}
       Scale.Y {type=MovieSceneScriptingDoubleChannel, default=1}
       Scale.Z {type=MovieSceneScriptingDoubleChannel, default=1}
      ]
     }]
   }
  ]
 }
 NiagaraComponent0 {
  id=#39
  parent_id=#14
  possessed_class=NiagaraComponent
  tracks [{
    type=MovieSceneNiagaraFloatParameterTrack
    display_name=Intensity
    sections [{
      type=MovieSceneFloatSection
      channels [None {
        type=MovieSceneScriptingFloatChannel
        default=1
        keys [[173,1],[244.1,0],[840,0.7],[1110,0.7],[1379,0.3],[1380,0.65],[1650,0.65],[2219,0.015],[2220,1],[2511.0825,1],[2682.935,0]]
       }]
     }]
   }]
 }
 NS_DragonCart_Shard_ChargeUp {
  id=#40
  possessed_class=NiagaraActor
  tracks [
   {
    type=MovieSceneSpawnTrack
    sections [{
      type=MovieSceneSpawnSection
      channels [None {type=MovieSceneScriptingBoolChannel, default=false, keys [[1928.9175,false],[2347.615,true],[2929.225,false]]}]
     }]
   }
   {
    type=MovieScene3DTransformTrack
    sections [{
      type=MovieScene3DTransformSection
      channels [
       Location.X {type=MovieSceneScriptingDoubleChannel, default=-11011}
       Location.Y {type=MovieSceneScriptingDoubleChannel, default=0}
       Location.Z {type=MovieSceneScriptingDoubleChannel, default=3936.7445}
       Rotation.X {type=MovieSceneScriptingDoubleChannel, default=0}
       Rotation.Y {type=MovieSceneScriptingDoubleChannel, default=0}
       Rotation.Z {type=MovieSceneScriptingDoubleChannel, default=0}
       Scale.X {type=MovieSceneScriptingDoubleChannel, default=1}
       Scale.Y {type=MovieSceneScriptingDoubleChannel, default=1}
       Scale.Z {type=MovieSceneScriptingDoubleChannel, default=1}
      ]
     }]
   }
  ]
 }
 NS_DragonCart_ShardShockWave {
  id=#31
  possessed_class=NiagaraActor
  tracks [
   {
    type=MovieSceneSpawnTrack
    sections [{
      type=MovieSceneSpawnSection
      channels [None {
        type=MovieSceneScriptingBoolChannel
        default=false
        keys [[838,false],[839,true],[950,false],[1379,false],[1380,true],[1466,false],[2219,false],[2220,true],[2346,false]]
       }]
     }]
   }
   {
    type=MovieScene3DTransformTrack
    sections [{
      type=MovieScene3DTransformSection
      channels [
       Location.X {type=MovieSceneScriptingDoubleChannel, default=-11011}
       Location.Y {type=MovieSceneScriptingDoubleChannel, default=0}
       Location.Z {type=MovieSceneScriptingDoubleChannel, default=3936.7445}
       Rotation.X {type=MovieSceneScriptingDoubleChannel, default=0}
       Rotation.Y {type=MovieSceneScriptingDoubleChannel, default=0}
       Rotation.Z {type=MovieSceneScriptingDoubleChannel, default=0}
       Scale.X {type=MovieSceneScriptingDoubleChannel, default=1}
       Scale.Y {type=MovieSceneScriptingDoubleChannel, default=1}
       Scale.Z {type=MovieSceneScriptingDoubleChannel, default=1}
      ]
     }]
   }
  ]
 }
 CameraShakeSourceActor {
  id=#36
  possessed_class=CameraShakeSourceActor
  tracks [
   {
    type=MovieSceneSpawnTrack
    sections [{
      type=MovieSceneSpawnSection
      channels [None {type=MovieSceneScriptingBoolChannel, default=false, keys [[372.205,false],[622.4,true],[3383.5975,false]]}]
     }]
   }
   {
    type=MovieScene3DTransformTrack
    sections [{
      type=MovieScene3DTransformSection
      channels [
       Location.X {type=MovieSceneScriptingDoubleChannel, default=-11011}
       Location.Y {type=MovieSceneScriptingDoubleChannel, default=0}
       Location.Z {type=MovieSceneScriptingDoubleChannel, default=5230.3754}
       Rotation.X {type=MovieSceneScriptingDoubleChannel, default=0}
       Rotation.Y {type=MovieSceneScriptingDoubleChannel, default=0}
       Rotation.Z {type=MovieSceneScriptingDoubleChannel, default=0}
       Scale.X {type=MovieSceneScriptingDoubleChannel, default=1}
       Scale.Y {type=MovieSceneScriptingDoubleChannel, default=1}
       Scale.Z {type=MovieSceneScriptingDoubleChannel, default=1}
      ]
     }]
   }
  ]
 }
 NiagaraComponent0 {
  id=#41
  parent_id=#18
  possessed_class=NiagaraComponent
  tracks [{
    type=MovieSceneNiagaraFloatParameterTrack
    display_name=Intensity
    sections [{
      type=MovieSceneFloatSection
      channels [None {
        type=MovieSceneScriptingFloatChannel
        default=0
        keys [[173,0],[244.1,0],[840,0.7],[1110,0.7],[1379,0.3],[1380,0.65],[1650,0.65],[2219,0.015],[2220,1],[2511.0825,1],[2682.935,0]]
       }]
     }]
   }]
 }
 NiagaraComponent0 {
  id=#42
  parent_id=#15
  possessed_class=NiagaraComponent
  tracks [
   {
    type=MovieSceneParticleTrack
    display_name=FX System
    sections [{
      type=MovieSceneParticleSection
      channels [None {type=MovieSceneScriptingParticleChannel, keys [[589, "<ParticleKey.DEACTIVATE: 1>"], [614, "<ParticleKey.ACTIVATE: 0>"]]}]
     }]
   }
   {type=MovieScene3DAttachTrack, sections [{type=MovieScene3DAttachSection}]}
   {
    type=MovieSceneNiagaraFloatParameterTrack
    display_name=Intensity
    sections [{
      type=MovieSceneFloatSection
      channels [None {
        type=MovieSceneScriptingFloatChannel
        default=1
        keys [[172,0],[243.1,0],[839,2.8],[1109,2.8],[1379,1],[1380,3],[2127,0],[2219,0],[2220,4],[2510.0825,2],[2681.935,0]]
       }]
     }]
   }
   {
    type=MovieSceneNiagaraFloatParameterTrack
    display_name=Lightning_Z_Offset
    sections [{
      type=MovieSceneFloatSection
      channels [None {type=MovieSceneScriptingFloatChannel, default=-4200, keys [[839,-2900],[917,-3200],[1425,-3500],[2275,-4200]]}]
     }]
   }
   {
    type=MovieSceneNiagaraFloatParameterTrack
    display_name=Length
    sections [{
      type=MovieSceneFloatSection
      channels [None {type=MovieSceneScriptingFloatChannel, default=1.35, keys [[839,1],[917,1.1],[1226,1.15],[1379,1.15],[1425,1.25],[2219,1.25],[2275,1.35]]}]
     }]
   }
   {
    type=MovieSceneNiagaraFloatParameterTrack
    display_name=UpLight_ExtraIntensity_Mult
    sections [{
      type=MovieSceneFloatSection
      channels [None {
        type=MovieSceneScriptingFloatChannel
        default=0.3
        keys [[839,1],[851,1],[879,0.65],[1380,1],[1403,1],[1429,0.65],[2218,0.5],[2219,1],[2267,1],[2298,0.65],[2510,0.3]]
       }]
     }]
   }
   {
    type=MovieSceneNiagaraFloatParameterTrack
    display_name=Core_Intensity
    sections [{
      type=MovieSceneFloatSection
      channels [None {
        type=MovieSceneScriptingFloatChannel
        default=2.8
        keys [[172,0],[243.1,0],[839,1.2],[1109,1.2],[1379,1],[1380,2.8],[2127,0],[2219,0],[2220,4],[2510.0825,2],[2681.935,0]]
       }]
     }]
   }
   {
    type=MovieSceneNiagaraFloatParameterTrack
    display_name=Core_Beam_Z_Offset
    sections [{
      type=MovieSceneFloatSection
      channels [None {
        type=MovieSceneScriptingFloatChannel
        default=533.2551
        keys [[838,263.9149],[839,533.2551],[1379,533.2551],[1380,687.0292],[1732,449.2933],[2220,263.9149],[2237,0],[2238,700.6893],[2281,463.1195],[2494,348.3141]]
       }]
     }]
   }
   {
    type=MovieSceneNiagaraBoolParameterTrack
    display_name=SpawnSmallLightningLFare
    sections [{type=MovieSceneBoolSection, channels [None {type=MovieSceneScriptingBoolChannel, default=true}]}]
   }
   {
    type=MovieSceneNiagaraFloatParameterTrack
    display_name=SmallLightning_FlareIntensity
    sections [{
      type=MovieSceneFloatSection
      channels [None {
        type=MovieSceneScriptingFloatChannel
        default=1
        keys [[172,0],[243.1,0],[839,2.8],[1109,2.8],[1379,1],[1380,3],[2127,0],[2219,0],[2220,4],[2510.0825,2],[2681.935,0]]
       }]
     }]
   }
  ]
 }
 NiagaraComponent0 {
  id=#43
  parent_id=#40
  possessed_class=NiagaraComponent
  tracks [{
    type=MovieSceneNiagaraFloatParameterTrack
    display_name=ChargeUp
    sections [{type=MovieSceneFloatSection, channels [None {type=MovieSceneScriptingFloatChannel, default=1, keys [[2347.615,0],[2897.2925,1]]}]}]
   }]
 }
 NiagaraComponent0 {
  id=#44
  parent_id=#45
  possessed_class=NiagaraComponent
  tracks [
   {
    type=MovieSceneNiagaraFloatParameterTrack
    display_name=ChargeUp
    sections [{
      type=MovieSceneFloatSection
      channels [None {
        type=MovieSceneScriptingFloatChannel
        default=1
        keys [[630,0],[789,1],[838,1],[839,0],[1110,0],[1257,1],[1380,1],[1393,0],[1650,0],[2220,1.5],[2221,0]]
       }]
     }]
   }
   {
    type=MovieSceneNiagaraFloatParameterTrack
    display_name=ChargeUp_Rate
    sections [{
      type=MovieSceneFloatSection
      channels [None {
        type=MovieSceneScriptingFloatChannel
        default=1
        keys [[630,0],[779,1],[839,1],[872,0],[1110,0],[1257,1],[1387,1],[1399,0],[1650,0],[2221,1.5],[2222,0]]
       }]
     }]
   }
   {type=MovieScene3DAttachTrack, sections [{type=MovieScene3DAttachSection}]}
  ]
 }
 NS_DragonCart_Shard_Electricity_01 {
  id=#23
  possessed_class=NiagaraActor
  tracks [
   {type=MovieSceneSpawnTrack, sections [{type=MovieSceneSpawnSection, channels [None {type=MovieSceneScriptingBoolChannel, default=true}]}]}
   {
    type=MovieScene3DTransformTrack
    sections [{
      type=MovieScene3DTransformSection
      channels [
       Location.X {type=MovieSceneScriptingDoubleChannel, default=-11174.1237}
       Location.Y {type=MovieSceneScriptingDoubleChannel, default=43.9823}
       Location.Z {type=MovieSceneScriptingDoubleChannel, default=3816.26}
       Rotation.X {type=MovieSceneScriptingDoubleChannel, default=4.7295}
       Rotation.Y {type=MovieSceneScriptingDoubleChannel, default=17.4221}
       Rotation.Z {type=MovieSceneScriptingDoubleChannel, default=-18.0857}
       Scale.X {type=MovieSceneScriptingDoubleChannel, default=1}
       Scale.Y {type=MovieSceneScriptingDoubleChannel, default=1}
       Scale.Z {type=MovieSceneScriptingDoubleChannel, default=1}
      ]
     }]
   }
  ]
 }
 NS_DragonCart_Orb_ChargeUp {
  id=#45
  possessed_class=NiagaraActor
  tracks [
   {type=MovieSceneSpawnTrack, sections [{type=MovieSceneSpawnSection, channels [None {type=MovieSceneScriptingBoolChannel, default=true}]}]}
   {
    type=MovieScene3DTransformTrack
    sections [{
      type=MovieScene3DTransformSection
      channels [
       Location.X {type=MovieSceneScriptingDoubleChannel, default=-5045}
       Location.Y {type=MovieSceneScriptingDoubleChannel, default=0}
       Location.Z {type=MovieSceneScriptingDoubleChannel, default=-5287.4807}
       Rotation.X {type=MovieSceneScriptingDoubleChannel, default=0}
       Rotation.Y {type=MovieSceneScriptingDoubleChannel, default=0}
       Rotation.Z {type=MovieSceneScriptingDoubleChannel, default=0}
       Scale.X {type=MovieSceneScriptingDoubleChannel, default=1}
       Scale.Y {type=MovieSceneScriptingDoubleChannel, default=1}
       Scale.Z {type=MovieSceneScriptingDoubleChannel, default=1}
      ]
     }]
   }
  ]
 }
 NiagaraComponent0 {
  id=#46
  parent_id=#47
  possessed_class=NiagaraComponent
  tracks [{
    type=MovieSceneNiagaraFloatParameterTrack
    display_name=Intensity
    sections [{
      type=MovieSceneFloatSection
      channels [None {
        type=MovieSceneScriptingFloatChannel
        default=0
        keys [[173,0],[244.1,0],[840,0.7],[1110,0.7],[1379,0.3],[1380,0.65],[1650,0.65],[2219,0.015],[2220,1],[2511.0825,1],[2682.935,0]]
       }]
     }]
   }]
 }
 NiagaraComponent0 {
  id=#48
  parent_id=#19
  possessed_class=NiagaraComponent
  tracks [{
    type=MovieSceneNiagaraFloatParameterTrack
    display_name=Intensity
    sections [{
      type=MovieSceneFloatSection
      channels [None {
        type=MovieSceneScriptingFloatChannel
        default=0.015
        keys [[173,0.4],[244.1,0],[840,0.7],[1110,0.7],[1379,0.3],[1380,0.65],[1650,0.65],[2219,0.015],[2220,1],[2511.0825,1],[2682.935,0]]
       }]
     }]
   }]
 }
 NiagaraComponent0 {
  id=#49
  parent_id=#50
  possessed_class=NiagaraComponent
  tracks [{
    type=MovieSceneNiagaraFloatParameterTrack
    display_name=Intensity
    sections [{
      type=MovieSceneFloatSection
      channels [None {
        type=MovieSceneScriptingFloatChannel
        default=1
        keys [[173,1],[244.1,0],[840,0.7],[1110,0.7],[1379,0.3],[1380,0.65],[1650,0.65],[2219,0.015],[2220,1],[2511.0825,1],[2682.935,0]]
       }]
     }]
   }]
 }
 NiagaraComponent0 {
  id=#51
  parent_id=#26
  possessed_class=NiagaraComponent
  tracks [{
    type=MovieSceneNiagaraFloatParameterTrack
    display_name=Intensity
    sections [{
      type=MovieSceneFloatSection
      channels [None {
        type=MovieSceneScriptingFloatChannel
        default=1
        keys [[173,1],[244.1,0],[840,0.7],[1110,0.7],[1379,0.3],[1380,0.65],[1650,0.65],[2219,0.015],[2220,1],[2511.0825,1],[2682.935,0]]
       }]
     }]
   }]
 }
 NiagaraComponent0 {
  id=#52
  parent_id=#53
  possessed_class=NiagaraComponent
  tracks [{
    type=MovieSceneNiagaraFloatParameterTrack
    display_name=Intensity
    sections [{
      type=MovieSceneFloatSection
      channels [None {
        type=MovieSceneScriptingFloatChannel
        default=0
        keys [[173,0],[244.1,0],[840,0.7],[1110,0.7],[1379,0.3],[1380,0.65],[1650,0.65],[2219,0.015],[2220,1],[2511.0825,1],[2682.935,0]]
       }]
     }]
   }]
 }
 NS_DragonCart_Shard_Electricity_11 {
  id=#47
  possessed_class=NiagaraActor
  tracks [
   {type=MovieSceneSpawnTrack, sections [{type=MovieSceneSpawnSection, channels [None {type=MovieSceneScriptingBoolChannel, default=true}]}]}
   {
    type=MovieScene3DTransformTrack
    sections [{
      type=MovieScene3DTransformSection
      channels [
       Location.X {type=MovieSceneScriptingDoubleChannel, default=-11051.2705}
       Location.Y {type=MovieSceneScriptingDoubleChannel, default=-1165.8531}
       Location.Z {type=MovieSceneScriptingDoubleChannel, default=2872.26}
       Rotation.X {type=MovieSceneScriptingDoubleChannel, default=34.0304}
       Rotation.Y {type=MovieSceneScriptingDoubleChannel, default=57.1185}
       Rotation.Z {type=MovieSceneScriptingDoubleChannel, default=93.0844}
       Scale.X {type=MovieSceneScriptingDoubleChannel, default=1}
       Scale.Y {type=MovieSceneScriptingDoubleChannel, default=1}
       Scale.Z {type=MovieSceneScriptingDoubleChannel, default=1}
      ]
     }]
   }
  ]
 }
 NS_DragonCart_Shard_Electricity_Long_DayOne8 {id=#54, possessed_class=NiagaraActor, tracks []}
 NS_DragonCart_Shard_Electricity_Long_DayOne9 {id=#50, possessed_class=NiagaraActor, tracks []}
 NiagaraComponent0 {
  id=#55
  parent_id=#54
  possessed_class=NiagaraComponent
  tracks [{
    type=MovieSceneNiagaraFloatParameterTrack
    display_name=Intensity
    sections [{
      type=MovieSceneFloatSection
      channels [None {
        type=MovieSceneScriptingFloatChannel
        default=1
        keys [[173,1],[244.1,0],[840,0.7],[1110,0.7],[1379,0.3],[1380,0.65],[1650,0.65],[2219,0.015],[2220,1],[2511.0825,1],[2682.935,0]]
       }]
     }]
   }]
 }
 NS_DragonCart_Shard_Gate E {
  id=#56
  parent_id=#4
  possessed_class=NiagaraComponent
  tracks [{
    type=MovieSceneParticleTrack
    display_name=FX System
    sections [{
      type=MovieSceneParticleSection
      channels [None {type=MovieSceneScriptingParticleChannel, keys [[0, "<ParticleKey.DEACTIVATE: 1>"], [2931, "<ParticleKey.ACTIVATE: 0>"]]}]
     }]
   }]
 }
 NiagaraComponent0 {
  id=#57
  parent_id=#16
  possessed_class=NiagaraComponent
  tracks [{
    type=MovieSceneNiagaraFloatParameterTrack
    display_name=Intensity
    sections [{
      type=MovieSceneFloatSection
      channels [None {
        type=MovieSceneScriptingFloatChannel
        default=0
        keys [[173,0],[244.1,0],[840,0.7],[1110,0.7],[1379,0.3],[1380,0.65],[1650,0.65],[2219,0.015],[2220,1],[2511.0825,1],[2682.935,0]]
       }]
     }]
   }]
 }
 NS_DragonCart_Shard_Electricity_9 {
  id=#53
  possessed_class=NiagaraActor
  tracks [
   {type=MovieSceneSpawnTrack, sections [{type=MovieSceneSpawnSection, channels [None {type=MovieSceneScriptingBoolChannel, default=true}]}]}
   {
    type=MovieScene3DTransformTrack
    sections [{
      type=MovieScene3DTransformSection
      channels [
       Location.X {type=MovieSceneScriptingDoubleChannel, default=-10082.5817}
       Location.Y {type=MovieSceneScriptingDoubleChannel, default=594.7192}
       Location.Z {type=MovieSceneScriptingDoubleChannel, default=2856.9155}
       Rotation.X {type=MovieSceneScriptingDoubleChannel, default=-0.5806}
       Rotation.Y {type=MovieSceneScriptingDoubleChannel, default=51.3822}
       Rotation.Z {type=MovieSceneScriptingDoubleChannel, default=-132.9066}
       Scale.X {type=MovieSceneScriptingDoubleChannel, default=1}
       Scale.Y {type=MovieSceneScriptingDoubleChannel, default=1}
       Scale.Z {type=MovieSceneScriptingDoubleChannel, default=1}
      ]
     }]
   }
  ]
 }
 NS_DragonCart_Shard_Electricity_5 {
  id=#34
  possessed_class=NiagaraActor
  tracks [
   {type=MovieSceneSpawnTrack, sections [{type=MovieSceneSpawnSection, channels [None {type=MovieSceneScriptingBoolChannel, default=true}]}]}
   {
    type=MovieScene3DTransformTrack
    sections [{
      type=MovieScene3DTransformSection
      channels [
       Location.X {type=MovieSceneScriptingDoubleChannel, default=-11601.0318}
       Location.Y {type=MovieSceneScriptingDoubleChannel, default=921.7044}
       Location.Z {type=MovieSceneScriptingDoubleChannel, default=2872.26}
       Rotation.X {type=MovieSceneScriptingDoubleChannel, default=175.2017}
       Rotation.Y {type=MovieSceneScriptingDoubleChannel, default=46.0561}
       Rotation.Z {type=MovieSceneScriptingDoubleChannel, default=-62.898}
       Scale.X {type=MovieSceneScriptingDoubleChannel, default=1}
       Scale.Y {type=MovieSceneScriptingDoubleChannel, default=1}
       Scale.Z {type=MovieSceneScriptingDoubleChannel, default=1}
      ]
     }]
   }
  ]
 }
]
display_rate {denominator=1, numerator=60}
tick_resolution {denominator=1, numerator=24000}
tracks [
 {
  type=MovieSceneEventTrack
  display_name=LoadKnockBackVolumes
  sections [{
    type=MovieSceneEventTriggerSection
    channels [None {
      type=MovieSceneScriptingEventChannel
      keys [
       [173, "<Struct 'MovieSceneEvent' (0x49e254614a0) {}>"]
       [843, "<Struct 'MovieSceneEvent' (0x49e254617c0) {}>"]
       [1388, "<Struct 'MovieSceneEvent' (0x49e25461b80) {}>"]
       [2227, "<Struct 'MovieSceneEvent' (0x49e25461cc0) {}>"]
       [2936, "<Struct 'MovieSceneEvent' (0x49e25461d60) {}>"]
      ]
     }]
   }]
 }
 {
  type=MovieSceneMaterialParameterCollectionTrack
  display_name=MPC_DragonCart_Orb_Rift
  sections [{
    type=MovieSceneParameterSection
    channels [
     OrbWPOScale {type=MovieSceneScriptingFloatChannel, default=1, keys [[133,1],[540,0],[3180.3175,0],[3303.8075,1]]}
     Orb Emissive Strength {
      type=MovieSceneScriptingFloatChannel
      default=40
      keys [[630,1],[840,10],[878.365,1],[1110,1],[1380,20],[1460.98,1],[1650,1],[2220,40],[2456.6325,1],[2489.925,1],[2495.905,0],[2501.415,1],[2508.7575,0],[2516.1975,1],[2530.52,1],[2534.9125,0],[2538.2325,1],[2541.905,1],[2545.745,0],[2548.79,1],[2551.255,0],[2554.0075,0],[2557.9725,1],[2578.08,1],[2581.005,0],[2584.415,1],[2589.9375,0.1],[2592.77,0],[2597.5325,0.1],[2601.5825,0],[2605.155,0.1],[2612.6825,0.1],[2617.375,0],[2621.13,0.1],[2625.085,0],[2631.7875,0],[2635.545,0.25],[2719.2625,0],[3060,0],[3065,0.0397],[3071,0],[3076,0.1045],[3084,0],[3091,0.1373],[3106,0.1136],[3110,0],[3113,0.1136],[3117,0.1136],[3121,0],[3124,0.1136],[3126,0],[3129,0],[3133,0.1172],[3153,0.0678],[3156,0],[3159,0.0715],[3165,0.0467],[3168,0],[3172,0.0467],[3177,0],[3180,0.0467],[3188,0.0467],[3192,0],[3196,0.0788],[3204,0.0319],[3212,0.0133],[3223,0.0339],[3247,0.05]]
     }
     ZPJitter {
      type=MovieSceneScriptingFloatChannel
      default=6
      keys [[173,1],[217.2725,0],[840,0],[851.965,4],[885,2],[1331.845,2],[1380,1],[1395.71,5],[1440,3],[1752.78,3],[2112.75,1],[2220,1],[2235,6],[2291,2],[2359.5075,2],[2944.225,6],[3032.64,1],[3303.8075,0]]
     }
     ZPOffset {
      type=MovieSceneScriptingFloatChannel
      default=15
      keys [[173,25],[217.2725,60],[840,60],[855,150],[885,90],[1380,20],[1395,190],[1440,110],[1808,0],[2220,100],[2235,250],[2291,120],[2719.2625,-20],[2929.225,-20],[2944.225,250],[3044,15]]
     }
    ]
   }]
 }
 {type=MovieSceneSubTrack, display_name=Subsequences, sections [{type=MovieSceneSubSection, end_frame=3840, start_frame=0}]}
]
